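# Material function for the MindPinch character's toon flipbook mouth, outputting an alpha mask that picks a face-pose frame from animation pose parameters.
# Thresholded, rounded smile, frown and jaw-open poses blend between closed and open frame indices, with lip-corner narrow overriding last.
# SubUV samples that frame on offset UV1, clipped by a floor-based UDIM tile mask, behind a static switch.
# Phoneme and upper-lip-raiser branches are left unconnected.
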MaterialFunction /BRCosmetics/Characters/Player/Male/Medium/Bodies/M_MED_MindPinch/Materials/MFC_ToonFlipbookMouth_MindPinch
graph {
 properties {description=10}
 nodes [
  n1 = TextureObjectParameter() {
   group=1 - ToonFlipbookMouth_MindPinch
   parameterName=ToonMouthFlipbookTexture
   samplerType=SAMPLERTYPE_Alpha
   sortPriority=1
   texture=/BRCosmetics/Characters/Player/Male/Medium/Bodies/M_MED_MindPinch/Textures/T_M_MED_MindPinch_FacePoses
  }
  n2 = VectorParameter() {desc=Mouth UDIM coord, group=1 - ToonFlipbookMouth_MindPinch, parameterName=Mouth UDIM coord, defaultValue=<rgba 0 1 0 1>}
  n3 = Add(A=TextureCoordinate() {coordinateIndex=1}, B=ComponentMask(None=@n2.RGB) {g=true, r=true})
  n4 = ScalarParameter() {defaultValue=2, desc=Smile Closed/Narrow, group=1 - ToonFlipbookMouth_MindPinch, parameterName=Smile Frame, sliderMax=16}
  n5 = ScalarParameter() {group=1 - Mouth Animation Poses, parameterName=R_smile_pose, sliderMax=1}
  n6 = ScalarParameter() {defaultValue=1, group=1 - ToonFlipbookMouth_MindPinch, parameterName=AnimationThreshold_Mouth, sliderMax=2}
  n7 = Saturate(None=Round(None=Multiply(A=@n5, B=@n6)))
  n8 = ScalarParameter() {defaultValue=4, desc=Frown Closed/Narrow, group=1 - ToonFlipbookMouth_MindPinch, parameterName=Frown Frame, sliderMax=16}
  n9 = ScalarParameter() {group=1 - Mouth Animation Poses, parameterName=R_Frown_pose, sliderMax=1}
  n10 = Saturate(None=Round(None=Multiply(A=@n9, B=@n6)))
  n11 = LinearInterpolate(
   A=LinearInterpolate(A=ScalarParameter() {group=1 - ToonFlipbookMouth_MindPinch, parameterName=Default Frame, sliderMax=16}, B=@n4, Alpha=@n7)
   B=@n8
   Alpha=@n10
  )
  n12 = LinearInterpolate(
   A=LinearInterpolate(
    A=ScalarParameter() {defaultValue=1, group=1 - ToonFlipbookMouth_MindPinch, parameterName=Default Open Frame, sliderMax=16}
    B=ScalarParameter() {defaultValue=3, group=1 - ToonFlipbookMouth_MindPinch, parameterName=Smile Open Frame, sliderMax=16}
    Alpha=@n7
   )
   B=ScalarParameter() {defaultValue=5, group=1 - ToonFlipbookMouth_MindPinch, parameterName=Frown Open Frame, sliderMax=16}
   Alpha=@n10
  )
  n13 = ScalarParameter() {group=1 - Mouth Animation Poses, parameterName=jaw_open_pose, sliderMax=1}
  n14 = Saturate(None=Round(None=Multiply(A=@n13, B=@n6)))
  n15 = ScalarParameter() {defaultValue=6, desc=Narrow / Closed, group=1 - ToonFlipbookMouth_MindPinch, parameterName=Narrow Frame, sliderMax=16}
  n16 = Round(None=Multiply(
    A=ScalarParameter() {desc="Closed/Narrow (Default)", group=1 - Mouth Animation Poses, parameterName=R_lip_corner_narrow_pose, sliderMax=1}
    B=@n6
   ))
  n17 = MaterialFunctionCall!(
   Texture=@n1
   UVs=@n3
   SubImages=ComponentMask(None=VectorParameter() {group=1 - ToonFlipbookMouth_MindPinch, parameterName=SubImages, defaultValue=<rgba 2 4 0 0>}.RGB) {g=true, r=true}
   Frame=LinearInterpolate(A=LinearInterpolate(A=@n11, B=@n12, Alpha=@n14) {desc=open close blend}, B=@n15, Alpha=Saturate(None=@n16))
  ) {materialFunction=/Engine/Functions/Engine_MaterialFunctions03/Texturing/SubUV_Function}
  n18 = Floor(None=@n3)
  n19 = MaterialFunctionCall(Float2=@n18) {materialFunction=&BreakOutFloat2Components}
  n20 = MaterialFunctionCall(Float2=Multiply(A=@n18) {constB=-1}) {materialFunction=&BreakOutFloat2Components}
  n21 = OneMinus(None=Max(A=Max(A=@n19.R, B=@n19.R) {constB=0}, B=Max(A=@n20.R, B=@n20.R) {constB=0}) {constB=0})
  FunctionOutput(None=StaticSwitchParameter(True=Min(A=@n17.Alpha, B=@n21), False=Constant()) {group=1 - ToonFlipbookMouth_MindPinch, parameterName=UseToonFlipbookMouth_MindPinch, sortPriority=0}) {outputName=DF}
  n22 = Multiply(A=ScalarParameter() {group=1 - Mouth Animation Poses, parameterName=phoneme_oo_pose, sliderMax=1})
  LinearInterpolate(
   B=ScalarParameter() {defaultValue=4, group=1 - ToonFlipbookMouth_MindPinch, parameterName=Phoneme OO Frame, sliderMax=16}
   Alpha=Saturate(None=Round(None=@n22))
  )
  n23 = Multiply(A=ScalarParameter() {group=1 - Mouth Animation Poses, parameterName=R_upper_lip_raiser_pose, sliderMax=1})
  LinearInterpolate(Alpha=Saturate(None=Round(None=@n23)))
  n24 = Multiply(A=ScalarParameter() {group=1 - Mouth Animation Poses, parameterName=L_upper_lip_raiser_pose, sliderMax=1})
  LinearInterpolate(Alpha=Saturate(None=Round(None=@n24)))
  n25 = Multiply(A=ScalarParameter() {group=1 - Mouth Animation Poses, parameterName=phoneme_mbp_pose, sliderMax=1})
  LinearInterpolate(
   B=ScalarParameter() {defaultValue=6, group=1 - ToonFlipbookMouth_MindPinch, parameterName=Phoneme MBP Frame, sliderMax=16}
   Alpha=Saturate(None=Round(None=@n25))
  )
 ]
}
