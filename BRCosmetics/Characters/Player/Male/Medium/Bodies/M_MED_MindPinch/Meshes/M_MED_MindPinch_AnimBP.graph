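# Post-process style Anim Blueprint for the MindPinch character body, layering cosmetic head and hand fixes over the parent mesh's copied pose.
# A Control Rig feeds head_offset rotation, damped from the head's ref-pose delta, then copies thumb and index rotations onto fake finger bones.
# Rifle and MindPinch pickaxe flags enable additive grip corrections, and left fake-thumb bones get scale tweaks.
# A frontend curve blends DefaultSlot animation onto hands and head.
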AnimBlueprint /BRCosmetics/Characters/Player/Male/Medium/Bodies/M_MED_MindPinch/Meshes/M_MED_MindPinch_AnimBP
target_skeleton=&/M_MED_MindPinch_Skeleton
graphs [
 AnimGraph {nodes [
   n1 = Layered blend per bone(BasePose=Use cached pose 'FinalPose'(), BlendPoses_0=Slot 'DefaultSlot'(), BlendWeights_0=Get CustomFEIdle() -> (CustomFEIdle:double)) {layerSetup [{branchFilters [{boneName=hand_l}, {boneName=hand_r}, {boneName=head_offset}]}]}
   Output Pose(Result=@n1) {layerGroup=None}
   n2 = Control Rig(Source=Copy Pose From Mesh() {bUseAttachedParent=true}) {
    alphaBoolBlend {blendOption=Linear}
    controlRigAssetReference {blueprintRigClass=&/M_MED_MindPinch_CtrlRig.M_MED_MindPinch_CtrlRig_C}
   }
   n3 = "Transform (Modify) Bone - Bone: head_offset"(ComponentPose=Local To Component(LocalPose=@n2), Alpha=1, Translation="0,0,0", Rotation=Get HeadRot() -> (HeadRot:Rotator), Scale="1,1,1") {rotationMode=BMM_Replace, rotationSpace=BCS_BoneSpace, alphaBoolBlend {blendOption=Linear}, boneToModify {boneName=head_offset}}
   n4 = Copy Bone Delta - Source Bone: thumb_03_l - Target Bone: thumb_fake_03_l!(
    ComponentPose:Component Space Pose Link Structure=Copy Bone Delta - Source Bone: thumb_02_l - Target Bone: thumb_fake_02_l!(
     ComponentPose:Component Space Pose Link Structure=@n3
     Alpha:float=1
     bCopyTranslation:bool=False
     bCopyRotation:bool=true
     bCopyScale:bool=False
    ) -> (Pose:Component Space Pose Link Structure) {
     is=title AnimGraphNode_CopyBoneDelta
     alphaBoolBlend {blendOption=Linear}
     sourceBone {boneName=thumb_02_l}
     targetBone {boneName=thumb_fake_02_l}
    }
    Alpha:float=1
    bCopyTranslation:bool=False
    bCopyRotation:bool=true
    bCopyScale:bool=False
   ) -> (Pose:Component Space Pose Link Structure) {
    is=title AnimGraphNode_CopyBoneDelta
    alphaBoolBlend {blendOption=Linear}
    sourceBone {boneName=thumb_03_l}
    targetBone {boneName=thumb_fake_03_l}
   }
   n5 = Copy Bone Delta - Source Bone: index_01_l - Target Bone: index_fake_01_l!(
    ComponentPose:Component Space Pose Link Structure=Copy Bone Delta - Source Bone: thumb_03_r - Target Bone: thumb_fake_03_r!(
     ComponentPose:Component Space Pose Link Structure=Copy Bone Delta - Source Bone: thumb_02_r - Target Bone: thumb_fake_02_r!(
      ComponentPose:Component Space Pose Link Structure=@n4
      Alpha:float=1
      bCopyTranslation:bool=False
      bCopyRotation:bool=true
      bCopyScale:bool=False
     ) -> (Pose:Component Space Pose Link Structure) {
      is=title AnimGraphNode_CopyBoneDelta
      alphaBoolBlend {blendOption=Linear}
      sourceBone {boneName=thumb_02_r}
      targetBone {boneName=thumb_fake_02_r}
     }
     Alpha:float=1
     bCopyTranslation:bool=False
     bCopyRotation:bool=true
     bCopyScale:bool=False
    ) -> (Pose:Component Space Pose Link Structure) {
     is=title AnimGraphNode_CopyBoneDelta
     alphaBoolBlend {blendOption=Linear}
     sourceBone {boneName=thumb_03_r}
     targetBone {boneName=thumb_fake_03_r}
    }
    Alpha:float=1
    bCopyTranslation:bool=False
    bCopyRotation:bool=true
    bCopyScale:bool=False
   ) -> (Pose:Component Space Pose Link Structure) {
    is=title AnimGraphNode_CopyBoneDelta
    alphaBoolBlend {blendOption=Linear}
    sourceBone {boneName=index_01_l}
    targetBone {boneName=index_fake_01_l}
   }
   n6 = Copy Bone Delta - Source Bone: index_01_r - Target Bone: index_fake_01_r!(
    ComponentPose:Component Space Pose Link Structure=Copy Bone Delta - Source Bone: index_03_l - Target Bone: index_fake_03_l!(
     ComponentPose:Component Space Pose Link Structure=Copy Bone Delta - Source Bone: index_02_l - Target Bone: index_fake_02_l!(
      ComponentPose:Component Space Pose Link Structure=@n5
      Alpha:float=1
      bCopyTranslation:bool=False
      bCopyRotation:bool=true
      bCopyScale:bool=False
     ) -> (Pose:Component Space Pose Link Structure) {
      is=title AnimGraphNode_CopyBoneDelta
      alphaBoolBlend {blendOption=Linear}
      sourceBone {boneName=index_02_l}
      targetBone {boneName=index_fake_02_l}
     }
     Alpha:float=1
     bCopyTranslation:bool=False
     bCopyRotation:bool=true
     bCopyScale:bool=False
    ) -> (Pose:Component Space Pose Link Structure) {
     is=title AnimGraphNode_CopyBoneDelta
     alphaBoolBlend {blendOption=Linear}
     sourceBone {boneName=index_03_l}
     targetBone {boneName=index_fake_03_l}
    }
    Alpha:float=1
    bCopyTranslation:bool=False
    bCopyRotation:bool=true
    bCopyScale:bool=False
   ) -> (Pose:Component Space Pose Link Structure) {
    is=title AnimGraphNode_CopyBoneDelta
    alphaBoolBlend {blendOption=Linear}
    sourceBone {boneName=index_01_r}
    targetBone {boneName=index_fake_01_r}
   }
   n7 = "Transform (Modify) Bone - Bone: thumb_fake_01_l"!(
    ComponentPose=Copy Bone Delta - Source Bone: index_03_r - Target Bone: index_fake_03_r!(
     ComponentPose:Component Space Pose Link Structure=Copy Bone Delta - Source Bone: index_02_r - Target Bone: index_fake_02_r!(
      ComponentPose:Component Space Pose Link Structure=@n6
      Alpha:float=1
      bCopyTranslation:bool=False
      bCopyRotation:bool=true
      bCopyScale:bool=False
     ) -> (Pose:Component Space Pose Link Structure) {
      is=title AnimGraphNode_CopyBoneDelta
      alphaBoolBlend {blendOption=Linear}
      sourceBone {boneName=index_02_r}
      targetBone {boneName=index_fake_02_r}
     }
     Alpha:float=1
     bCopyTranslation:bool=False
     bCopyRotation:bool=true
     bCopyScale:bool=False
    ) -> (Pose:Component Space Pose Link Structure) {
     is=title AnimGraphNode_CopyBoneDelta
     alphaBoolBlend {blendOption=Linear}
     sourceBone {boneName=index_03_r}
     targetBone {boneName=index_fake_03_r}
    }
    bAlphaBoolEnabled:bool=Get isRifle() -> (isRifle:bool)
    Translation="0,2,0"
    Rotation="0,5,0"
    Scale="1,1,1"
   ) -> (Pose) {
    alphaInputType=Bool
    rotationMode=BMM_Additive
    rotationSpace=BCS_BoneSpace
    scaleMode=BMM_Additive
    scaleSpace=BCS_BoneSpace
    translationMode=BMM_Additive
    translationSpace=BCS_BoneSpace
    alphaBoolBlend {blendOption=Linear}
    boneToModify {boneName=thumb_fake_01_l}
    rotation {yaw=5}
    translation {y=2}
   }
   n8 = Get isCustomPickaxe() -> (isCustomPickaxe:bool)
   n9 = "Transform (Modify) Bone - Bone: hand_r"!(ComponentPose=@n7, bAlphaBoolEnabled:bool=@n8, Translation="0.634,-0.008,-0.952", Rotation="-3.400,-3.947,-0.000", Scale="1,1,1") -> (Pose) {
    alphaInputType=Bool
    rotationMode=BMM_Additive
    rotationSpace=BCS_BoneSpace
    scaleSpace=BCS_BoneSpace
    translationMode=BMM_Additive
    translationSpace=BCS_BoneSpace
    alphaBoolBlend {blendOption=Linear}
    boneToModify {boneName=hand_r}
    rotation=<rot -3.4 -3.947 0>
    translation=<0.634 -0.008 -0.952>
   }
   n10 = "Transform (Modify) Bone - Bone: thumb_fake_01_r"!(ComponentPose=@n9, bAlphaBoolEnabled:bool=@n8, Translation="0,0,0", Rotation="1.162,23.854,-1.296", Scale="1.000,0.902,1.185") -> (Pose) {
    alphaInputType=Bool
    rotationMode=BMM_Additive
    rotationSpace=BCS_BoneSpace
    scaleMode=BMM_Additive
    scaleSpace=BCS_BoneSpace
    translationMode=BMM_Additive
    translationSpace=BCS_BoneSpace
    alphaBoolBlend {blendOption=Linear}
    boneToModify {boneName=thumb_fake_01_r}
    rotation=<rot 1.162 23.854 -1.296>
    scale {y=0.902, z=1.185}
   }
   n11 = "Transform (Modify) Bone - Bone: thumb_fake_02_r"!(ComponentPose=@n10, bAlphaBoolEnabled:bool=@n8, Translation="1.664,-0.113,-0.123", Rotation="-5.817,8.573,3.763", Scale="1,1,1") -> (Pose) {
    alphaInputType=Bool
    rotationMode=BMM_Additive
    rotationSpace=BCS_BoneSpace
    scaleMode=BMM_Additive
    scaleSpace=BCS_BoneSpace
    translationMode=BMM_Additive
    translationSpace=BCS_BoneSpace
    alphaBoolBlend {blendOption=Linear}
    boneToModify {boneName=thumb_fake_02_r}
    rotation=<rot -5.817 8.573 3.763>
    translation=<1.664 -0.113 -0.123>
   }
   n12 = "Transform (Modify) Bone - Bone: thumb_fake_03_r"!(ComponentPose=@n11, bAlphaBoolEnabled:bool=@n8, Translation="1.051,-0.088,-0.135", Rotation="-18.064,-33.722,23.797", Scale="1,1,1") -> (Pose) {
    alphaInputType=Bool
    rotationMode=BMM_Additive
    rotationSpace=BCS_BoneSpace
    scaleSpace=BCS_BoneSpace
    translationMode=BMM_Additive
    translationSpace=BCS_BoneSpace
    alphaBoolBlend {blendOption=Linear}
    boneToModify {boneName=thumb_fake_03_r}
    rotation=<rot -18.064 -33.722 23.797>
    translation=<1.051 -0.088 -0.135>
   }
   n13 = "Transform (Modify) Bone - Bone: index_fake_01_r"!(ComponentPose=@n12, bAlphaBoolEnabled:bool=@n8, Translation="1.094,-0.613,0.128", Rotation="21.472,12.618,-14.868", Scale="1,1,1") -> (Pose) {
    alphaInputType=Bool
    rotationMode=BMM_Additive
    rotationSpace=BCS_BoneSpace
    scaleSpace=BCS_BoneSpace
    translationMode=BMM_Additive
    translationSpace=BCS_BoneSpace
    alphaBoolBlend {blendOption=Linear}
    boneToModify {boneName=index_fake_01_r}
    rotation=<rot 21.472 12.618 -14.868>
    translation=<1.094 -0.613 0.128>
   }
   n14 = "Transform (Modify) Bone - Bone: index_fake_02_r"!(ComponentPose=@n13, bAlphaBoolEnabled:bool=@n8, Translation="-0.878,-0.718,-0.000", Rotation="-0.000,39.246,0.000", Scale="1,1,1") -> (Pose) {
    alphaInputType=Bool
    rotationMode=BMM_Additive
    rotationSpace=BCS_BoneSpace
    scaleSpace=BCS_BoneSpace
    translationMode=BMM_Additive
    translationSpace=BCS_BoneSpace
    alphaBoolBlend {blendOption=Linear}
    boneToModify {boneName=index_fake_02_r}
    rotation=<rot 0 39.246 0>
    translation=<-0.8776 -0.7176 0>
   }
   n15 = "Transform (Modify) Bone - Bone: index_fake_03_r"!(ComponentPose=@n14, bAlphaBoolEnabled:bool=@n8, Translation="-0.583,1.143,-0.000", Rotation="-0.000,-63.045,0.000", Scale="1,1,1") -> (Pose) {
    alphaInputType=Bool
    rotationMode=BMM_Additive
    rotationSpace=BCS_BoneSpace
    scaleSpace=BCS_BoneSpace
    translationMode=BMM_Additive
    translationSpace=BCS_BoneSpace
    alphaBoolBlend {blendOption=Linear}
    boneToModify {boneName=index_fake_03_r}
    rotation=<rot 0 -63.045 0>
    translation=<-0.583 1.143>
   }
   n16 = "Transform (Modify) Bone - Bone: ring_01_r"!(ComponentPose=@n15, bAlphaBoolEnabled:bool=@n8, Translation="0.753,-0.309,-0.190", Rotation="13.219,14.841,7.194", Scale="1,1,1") -> (Pose) {
    alphaInputType=Bool
    rotationMode=BMM_Additive
    rotationSpace=BCS_BoneSpace
    scaleSpace=BCS_BoneSpace
    translationMode=BMM_Additive
    translationSpace=BCS_BoneSpace
    alphaBoolBlend {blendOption=Linear}
    boneToModify {boneName=ring_01_r}
    rotation=<rot 13.219 14.841 7.194>
    translation=<0.7531 -0.3094 -0.19>
   }
   n17 = "Transform (Modify) Bone - Bone: ring_02_r"!(ComponentPose=@n16, bAlphaBoolEnabled:bool=@n8, Translation="-0.841,-0.539,-0.000", Rotation="7.813,32.548,0.007", Scale="1,1,1") -> (Pose) {
    alphaInputType=Bool
    rotationMode=BMM_Additive
    rotationSpace=BCS_BoneSpace
    scaleSpace=BCS_BoneSpace
    translationMode=BMM_Additive
    translationSpace=BCS_BoneSpace
    alphaBoolBlend {blendOption=Linear}
    boneToModify {boneName=ring_02_r}
    rotation=<rot 7.813 32.548 0.007>
    translation=<-0.8409 -0.5389 0>
   }
   n18 = "Transform (Modify) Bone - Bone: ring_03_r"!(ComponentPose=@n17, bAlphaBoolEnabled:bool=@n8, Translation="-0.396,0.557,1.451", Rotation="0.015,-55.312,0.021", Scale="1,1,1") -> (Pose) {
    alphaInputType=Bool
    rotationMode=BMM_Additive
    rotationSpace=BCS_BoneSpace
    scaleSpace=BCS_BoneSpace
    translationMode=BMM_Additive
    translationSpace=BCS_BoneSpace
    alphaBoolBlend {blendOption=Linear}
    boneToModify {boneName=ring_03_r}
    rotation=<rot 0.015 -55.312 0.021>
    translation=<-0.396 0.557 1.451>
   }
   n19 = "Transform (Modify) Bone - Bone: thumb_fake_01_l"!(ComponentPose=@n18, Alpha:float=1, Translation="0,0,0", Rotation="0,0,0", Scale="0.85,0.85,0.85") -> (Pose) {
    rotationMode=BMM_Additive
    rotationSpace=BCS_BoneSpace
    scaleMode=BMM_Additive
    scaleSpace=BCS_BoneSpace
    translationMode=BMM_Additive
    translationSpace=BCS_BoneSpace
    alphaBoolBlend {blendOption=Linear}
    boneToModify {boneName=thumb_fake_01_l}
    scale=<0.85 0.85 0.85>
   }
   n20 = "Transform (Modify) Bone - Bone: thumb_fake_02_l"(ComponentPose=@n19, Alpha=1, Translation="0,0,0", Rotation="0,0,0", Scale="0.85,0.85,0.85") {
    rotationMode=BMM_Additive
    rotationSpace=BCS_BoneSpace
    scaleMode=BMM_Additive
    scaleSpace=BCS_BoneSpace
    translationMode=BMM_Additive
    translationSpace=BCS_BoneSpace
    alphaBoolBlend {blendOption=Linear}
    boneToModify {boneName=thumb_fake_02_l}
    scale=<0.85 0.85 0.85>
   }
   n21 = "Transform (Modify) Bone - Bone: thumb_fake_03_l"(ComponentPose=@n20, Alpha=1, Translation="0,0,0", Rotation="0,0,0", Scale="0.85,0.85,0.85") {
    rotationMode=BMM_Additive
    rotationSpace=BCS_BoneSpace
    scaleSpace=BCS_BoneSpace
    translationMode=BMM_Additive
    translationSpace=BCS_BoneSpace
    alphaBoolBlend {blendOption=Linear}
    boneToModify {boneName=thumb_fake_03_l}
    scale=<0.85 0.85 0.85>
   }
   Save cached pose 'FinalPose'(Pose=Component To Local(ComponentPose=@n21)) {cacheName=FinalPose}
  ]}
 EventGraph {nodes [
   Event BlueprintUpdateAnimation()
   > n1 = Sequence!(execute) -> (then_0, then_1, then_2:Exec)
   > n2 = Branch(Condition=Get bIsCrouching())
   > Set HeadRotYawPCT(HeadRotYawPCT:double=0.8) -> (Output_Get:double)
   Set HeadRotYawPCT(execute=@n2.else, HeadRotYawPCT=0.6)
   n3 = GetOwnerSkeletalMeshComponent()
   n4 = BreakRotator(InRot=BreakTransform(InTransform=GetDeltaTransformFromRefPose(self=@n3, BoneName=head, BaseName=spine_05)).Rotation)
   n5 = Get HeadRotYawPCT() -> (HeadRotYawPCT:double)
   Set HeadRot(
    execute=@n1.then_1
    HeadRot:Rotator=MakeRotator(
     Roll=float * float(A=@n4, B=0.75) {is=title K2Node_PromotableOperator}
     Pitch=float * float(A=@n4.Pitch, B=0.5)
     Yaw=float * float(A=@n4.Yaw, B=@n5)
    )
   ) -> (Output_Get:Rotator)
   n6 = Get CurrentWeapon!(self:Fort Pawn Object Reference=TryGetFortPawn()) -> (CurrentWeapon:Fort Weapon Object Reference)
   > n7 = Branch(Condition="Equal (Enum)"(
     A:EFortWeaponCoreAnimation Enum=GetCoreAnimation!(self:Fort Weapon Object Reference=@n6) -> (ReturnValue:EFortWeaponCoreAnimation Enum)
     B:EFortWeaponCoreAnimation Enum=AssaultRifle
    ))
   > Set isRifle(isRifle:bool=true) -> (Output_Get:bool)
   Set isRifle(execute=@n7.else, isRifle=false)
   > n8 = Branch(Condition=PickaxeItemDefinitionHasAnyMetaTag(Tags=Make Array("[0]"=(TagName="Cosmetics.MetaTags.MindPinch")), bRequiresEquipped=true))
   > Set isCustomPickaxe(isCustomPickaxe:bool=true) -> (Output_Get:bool)
   Set isCustomPickaxe(execute=@n8.else, isCustomPickaxe=false)
   > Branch(Condition=Get bIsInFrontEnd())
   > Set CustomFEIdle(CustomFEIdle:double=GetCurveValue(self=GetOwnerFrontEndAnimInstance(), CurveName=Frontend_Idle_Bypass)) -> (Output_Get:double)
   Event OnBeginTargeting() {is=title K2Node_Event}
   > n9 = Cast To FortWeaponRanged!(execute, Object=Get CurrentWeapon!() -> (CurrentWeapon)) -> (then, CastFailed, AsFort Weapon Ranged:Fort Weapon Ranged Object Reference) {is=title K2Node_DynamicCast}
   > n10 = Branch(Condition=IsUsingScopeOrFirstPersonCamera!(self:Fort Weapon Ranged Object Reference=@n9) -> (ReturnValue:bool))
   > Set MoveThumbInScoped(MoveThumbInScoped:double=1) -> (Output_Get:double)
   Set MoveThumbInScoped(execute=@n10.else, MoveThumbInScoped=0)
   Event OnEndTargeting() {is=title K2Node_Event}
   > Set MoveThumbInScoped(MoveThumbInScoped=10)
  ]}
]
variables [
 CustomFEIdle {type=number}
 FortPawn {type=FortPawn, default=null}
 HeadRot {type=Rotator}
 HeadRotYawPCT {type=number, default=0.6}
 isCustomPickaxe {type=boolean}
 isRifle {type=boolean}
 mindPinch_pickaxe {type=number}
 MoveThumbInScoped {type=number}
 weapon_visible {type=boolean}
]
