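# Animation Blueprint for the MindPinch male character's face accessory mesh, posing it with the attached body and layering in slot animation.
# AnimGraph blends two copied poses (attached parent with curves, BodyMesh) at head_offset, then layers DefaultSlot there, weighted by CustomFEIdle.
# BodyMesh is cached when customization loads; CustomFEIdle reads a front-end curve whose name is unset.
# Each update also derives a scaled HeadRot from head-to-neck delta, unused in this AnimGraph.
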AnimBlueprint /BRCosmetics/Characters/Player/Male/Medium/Bodies/M_MED_MindPinch/Meshes/Parts/M_MED_MindPinch_FaceAcc_AnimBP
target_skeleton=&/M_MED_MindPinch_FaceAcc_Skeleton
graphs [
 AnimGraph {nodes [
   n1 = Copy Pose From Mesh(SourceMeshComponent=Get BodyMesh() -> (BodyMesh:Skeletal Mesh Component Object Reference)) {bUseAttachedParent=true}
   n2 = Copy Pose From Mesh() {bCopyCurves=true, bUseAttachedParent=true}
   Output Pose(Result=Layered blend per bone(
     BasePose=Layered blend per bone(BasePose=@n2, BlendPoses_0=@n1) {layerSetup [{branchFilters [{boneName=head_offset}]}]}
     BlendPoses_0=Slot 'DefaultSlot'()
     BlendWeights_0=Get CustomFEIdle() -> (CustomFEIdle:double)
    ) {layerSetup [{branchFilters [{boneName=head_offset}]}]}) {layerGroup=None}
  ]}
 EventGraph {nodes [
   Event BlueprintUpdateAnimation()
   n1 = GetOwnerSkeletalMeshComponent()
   n2 = BreakRotator(InRot=BreakTransform(InTransform=GetDeltaTransformFromRefPose(self=@n1, BoneName=head, BaseName=neck_02)).Rotation)
   > Set HeadRot(HeadRot:Rotator=MakeRotator(
     Roll=float * float(A=@n2, B=1) {is=title K2Node_PromotableOperator}
     Pitch=float * float(A=@n2.Pitch, B=1)
     Yaw=float * float(A=@n2.Yaw, B=0.5)
    )) -> (Output_Get:Rotator)
   > Branch(Condition=Get bIsInFrontEnd())
   > Set CustomFEIdle(CustomFEIdle:double=GetCurveValue(self=GetOwnerFrontEndAnimInstance(), CurveName=None)) -> (Output_Get:double)
   Event OnCharacterCustomizationLoaded()
   > n3 = Cast To PlayerPawn_Athena(Object=TryGetPawnOwner())
   > Set BodyMesh(BodyMesh=GetSkeletalMeshForPartType(self=@n3))
  ]}
]
variables [
 BodyMesh {type=SkeletalMeshComponent}
 CustomFEIdle {type=number}
 HeadRot {type=Rotator}
]
